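# Animation Blueprint for the MixedAlley medium male player body, posing its tail and backpack attachment over a pose copied from the attached parent mesh.
# A Control Rig takes the copied pose and knee-sliding state, then additive bone modifiers adjust base_tail and lower attach_backpack.
# The event graph tracks backbling, cape, LOD, hang-gliding and crouch-driven tail-correction flags.
# It also detects a matching MixedAlley backpack by wildcard matching the mesh component's name.
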AnimBlueprint /BRCosmetics/Characters/Player/Male/Medium/Bodies/M_MED_MixedAlley/Meshes/M_MED_MixedAlley_AnimBP
target_skeleton=&/M_MED_MixedAlley_Skeleton
graphs [
 AnimGraph {nodes [
   n1 = "Transform (Modify) Bone - Bone: base_tail"!(
    ComponentPose=Local To Component(LocalPose=Use cached pose 'Ctrl Rig'())
    bAlphaBoolEnabled:bool=True
    Translation="0,0,0"
    Rotation="0,0,-30"
    Scale="0.5,1,1"
   ) -> (Pose) {
    alphaInputType=Bool
    rotationMode=BMM_Additive
    scaleMode=BMM_Additive
    scaleSpace=BCS_BoneSpace
    alphaBoolBlend {blendInTime=0.1, blendOption=Linear, blendOutTime=0.12}
    boneToModify {boneName=base_tail}
    rotation {roll=-30}
    scale {x=0.5}
   }
   n2 = "Transform (Modify) Bone - Bone: attach_backpack"!(ComponentPose=@n1, bAlphaBoolEnabled:bool=True, Translation="0,0,-5", Rotation="0,0,0", Scale="1,1,1") -> (Pose) {
    alphaInputType=Bool
    rotationMode=BMM_Additive
    scaleMode=BMM_Additive
    scaleSpace=BCS_BoneSpace
    translationMode=BMM_Additive
    alphaBoolBlend {blendInTime=0.1, blendOption=Linear, blendOutTime=0.12}
    boneToModify {boneName=attach_backpack}
    translation {z=-5}
   }
   Output Pose(Result=Component To Local(ComponentPose=@n2)) {layerGroup=None}
   n3 = Control Rig!(
    Source=Copy Pose From Mesh() {bUseAttachedParent=true}
    Alpha
    isWearingBackbling:bool=false
    LowLods:bool=false
    isHangGliding:bool=false
    Tail Corrective:bool=false
    isWearingShoes:bool=false
    IsKneeSliding:bool=Get bIsKneeSliding()
   ) -> (Pose) {controlRigAssetReference {blueprintRigClass=&/M_MED_Mixed_Alley_Ctrl_Rig.M_MED_Mixed_Alley_Ctrl_Rig_C}}
   Save cached pose 'Ctrl Rig'(Pose=@n3) {cacheName=Ctrl Rig}
  ]}
 EventGraph {nodes [
   Event BlueprintUpdateAnimation()
   > n1 = Sequence()
   n2 = CheckForBackbling()
   Set Is Wearing Back Item(execute=@n1.then_1, Is Wearing Back Item:bool=OR Boolean!(A=@n2, B=@n2.bIsBackblingACape, C:bool=@n2.bIsBackblingLong) -> (ReturnValue)) -> (Output_Get:bool)
   > Set is Wearing Cape(is Wearing Cape:bool=OR Boolean!(A=@n2.bIsBackblingACape, B=@n2.bIsBackblingLong) -> (ReturnValue)) -> (Output_Get:bool)
   > Set LowLODs(LowLODs:bool="integer < integer"(A=Get CurrentLOD(), B=2)) -> (Output_Get:bool)
   n3 = Cast To FortPlayerPawn(execute=@n1.then_0, Object=TryGetPawnOwner())
   n4 = GetSkeletalMeshForPartType(self=@n3, PartType=Backpack)
   > Set MixedAlleyPack(MixedAlleyPack:bool=MatchesWildcard(SourceString=GetDisplayName(Object=@n4), Wildcard=*MixedAlley*)) -> (Output_Get:bool)
   n5 = Event OnCharacterCustomizationLoaded()
   > Set Backpack Mesh(Backpack Mesh:Skeletal Mesh Component Object Reference=Get Array Item with EFort Custom Part Type(Array=@n5.CustomMeshParts, PartType=Backpack)) -> (Output_Get:Skeletal Mesh Component Object Reference)
   n6 = Get Backpack Mesh() -> (Backpack Mesh:Skeletal Mesh Component Object Reference)
   > Set MixedAlleyPack(MixedAlleyPack=MatchesWildcard(SourceString=GetDisplayName(Object=@n6), Wildcard=*MixedAlley*))
   Event OnParachuteClosed()
   > Set Is Hang Gliding(Is Hang Gliding:bool=false) -> (Output_Get:bool)
   Event OnEndCrouching()
   > Set Tail Corrective(Tail Corrective:bool=false) -> (Output_Get:bool)
   Event OnBeginCrouching()
   > Set Tail Corrective(Tail Corrective=true)
   Event OnParachuteOpened()
   > Switch on EGliderType(Selection=Get PlayerGliderType() -> (PlayerGliderType=HangGlider))
   > Set Is Hang Gliding(Is Hang Gliding=true)
  ]}
]
variables [
 Backpack Mesh {type=SkeletalMeshComponent}
 Is Hang Gliding {type=boolean}
 Is Wearing Back Item {type=boolean}
 is Wearing Cape {type=boolean}
 isWearingShoes {type=boolean}
 LowLODs {type=boolean, default=true}
 MixedAlleyPack {type=boolean}
 Tail Corrective {type=boolean}
]
